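# Material function for the WarmShadeWagon cosmetic that layers audio-reactive glow and logo effects onto incoming material attributes for the character body or its backpack.
# MF_AudioReactivity amplitude drives a mask built from FX-mask and Fresnel terms.
# An optional pre-skinned sphere mask spreads the effect radially, sized by sound, including Blueprint-fed AccumulatedSound.
# The IsBackpack switch changes base color and emissive, with procedural screen, scanline and panning-line effects built from a backpack mask texture.
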MaterialFunction /BRCosmetics/Characters/Player/Male/Large/Bodies/M_LRG_WarmShadeWagon/Materials/MF_WarmShadeWagon
graph {nodes [
  n1 = NamedRerouteUsage()
  n2 = LinearInterpolate(
   A=NamedRerouteUsage()
   B=VectorParameter() {group=WarmShadeWagon, parameterName=LogoColor, defaultValue=<rgba 0.6939 0.7605 0.855 1>}.RGB
   Alpha=NamedRerouteDeclaration(None=ChannelMaskParameter(None=@n1) {group=WarmShadeWagon, maskChannel=Green, parameterName=LogoFXMask, defaultValue=<rgba 0 1 0 0>}.R) {name=LogoMask, nodeColor=<rgba 0.9219 1 0 1>}
  )
  n3 = NamedRerouteUsage()
  n4 = NamedRerouteUsage()
  n5 = NamedRerouteUsage()
  n6 = LinearInterpolate(
   A=StaticSwitchParameter(True=LinearInterpolate(A=@n2, B=@n3, Alpha=@n4), False=NamedRerouteUsage()) {defaultValue=true, group=WarmShadeWagon, parameterName=UseSphereMask}
   Alpha=Saturate(None=Multiply(A=@n5) {constB=10})
  ) {constB=0}
  n7 = NamedRerouteUsage()
  n8 = LinearInterpolate(A=NamedRerouteUsage(), B=NamedRerouteUsage(), Alpha=NamedRerouteUsage()) {constB=0}
  n9 = Multiply(A=VectorParameter() {group=WarmShadeWagon, parameterName=WS_ReactivityColor, defaultValue=<rgba 20 0 0.35 1>}.RGB, B=NamedRerouteUsage())
  n10 = NamedRerouteUsage()
  n11 = Multiply(A=NamedRerouteUsage(), B=Constant() {r=200}) {constB=10}
  n12 = SetMaterialAttributes!(
   MaterialAttributes=NamedRerouteUsage()
   BaseColor=StaticSwitchParameter(True=LinearInterpolate(A=@n6, Alpha=@n7) {constB=0}, False=@n6) {group=WarmShadeWagon, parameterName=IsBackpack}
   EmissiveColor=StaticSwitchParameter(True=Add(A=@n8, B=@n9), False=Add(A=@n10, B=@n9)) {group=WarmShadeWagon, parameterName=IsBackpack}
   Opacity=Saturate(None=OneMinus(None=@n11)) {desc=Override SSS for Epic}
   CustomData0=NamedRerouteUsage()
   Metallic=NamedRerouteUsage()
   Roughness=NamedRerouteUsage()
  ) {attributeSetTypes [
    #1
    #2
    #3
    #4
    #5
    #6
   ]}
  FunctionOutput(None=@n12)
  LinearInterpolate(Alpha=NamedRerouteUsage()) {constB=0}
  n13 = NamedRerouteUsage()
  Multiply(A=@n13, B=NamedRerouteUsage())
  n14 = MaterialFunctionCall() {materialFunction=/Game/Packages/Fortress_SharedMaterials/Base_Material_Functions/BaseCharacterShaderFunctions/Audio/MF_AudioReactivity}
  Multiply(A=OneMinus(None=ScalarParameter() {defaultValue=1, group=WarmShadeWagon, parameterName=FrontendPreview, sliderMax=1}), B=@n14.Amplitude)
  NamedRerouteDeclaration(None=FunctionInput() {bUsePreviewValueAsDefault=true, inputType=FunctionInput_MaterialAttributes}) {name=InMaterialAttributes, nodeColor=<rgba 0 0.9297 1 1>}
  NamedRerouteDeclaration(None=FunctionInput() {bUsePreviewValueAsDefault=true, inputName=FXMask, sortPriority=1}) {name=InFXMask, nodeColor=<rgba 1 0.75 0 1>}
  n15 = SphereMask(A=Add(A=LocalPosition() {localOrigin=InstancePreSkinning}.XYZ, B=Constant3Vector() {constant=<rgba 0 -10 -155 0>}), B=Constant()) {attenuationRadius=10, hardnessPercent=50}
  n16 = ScalarParameter() {defaultValue=8, group=WarmShadeWagon, parameterName=FaceMaskExp}
  n17 = Saturate(None=Multiply(A=Power(Base=@n15, Exp=@n16), B=ScalarParameter() {defaultValue=10, group=WarmShadeWagon, parameterName=FaceMaskIntensity}) {constB=2})
  NamedRerouteDeclaration(None=StaticSwitchParameter(True=VertexInterpolator(VS=@n17).PS, False=Constant()) {group=WarmShadeWagon, parameterName=UseFaceMask}) {name=FaceMask, nodeColor=<rgba 0.2656 0 1 1>}
  n18 = Add(A=TextureCoordinate() {coordinateIndex=1}, B=Constant2Vector() {g=-0.375})
  n19 = ComponentMask(None=@n18) {r=true}
  n20 = ComponentMask(None=@n18) {g=true}
  n21 = Multiply(A=Time()) {constB=0.3}
  n22 = LinearInterpolate(Alpha=ConstantBiasScale(None=Sine(None=@n21))) {constA=0.5}
  n23 = Constant() {r=0.5}
  n24 = Subtract(A=Add(A=Divide(A=@n20, B=@n22), B=@n23) {constB=0.5}, B=Divide(A=@n23, B=@n22))
  n25 = TextureSample!(UVs=Panner(Coordinate=AppendVector(A=@n19, B=@n24)) {speedX=0.1}, Tex, Bias, Apply View MipBias) {
   mipValueMode=TMVM_MipBias
   samplerSource=SSM_Wrap_WorldGroupSettings
   samplerType=SAMPLERTYPE_LinearColor
   texture=/BRCosmetics/Accessories/FORT_Backpacks/Backpack_M_MED_WarmShadeWagon/Materials/T_WarmShadeWagon_Backpack_mask
  }
  n26 = TextureSample(UVs=Panner(Coordinate=@n18) {speedX=-0.15}) {
   mipValueMode=TMVM_MipBias
   samplerSource=SSM_Wrap_WorldGroupSettings
   samplerType=SAMPLERTYPE_LinearColor
   texture=&T_WarmShadeWagon_Backpack_mask
  }
  n27 = Multiply(A=@n26.B, B=@n26.B)
  n28 = Add(A=SmoothStep(Value=Multiply(A=@n25.G, B=@n25.G)) {constMin=0.85}, B=Multiply(A=SmoothStep(Value=@n27) {constMin=0.9}) {constB=0.3})
  n29 = MaterialFunctionCall!(
   A=TextureCoordinate() {coordinateIndex=1}
   B=Constant2Vector() {g=0.875, r=0.51}
   Bounds=Constant2Vector() {g=0.2, r=0.96}
   Edge Falloff=Constant()
  ) {materialFunction=/Engine/Functions/Engine_MaterialFunctions01/Gradient/BoxMask-2D}
  NamedRerouteDeclaration(None=Multiply(A=Multiply(A=@n28) {constB=0.8}, B=NamedRerouteDeclaration(None=@n29.Result) {name=LinesMask, nodeColor=<rgba 1 0 0.4453 1>})) {name=Lines, nodeColor=<rgba 1 0 0.2812 1>}
  n30 = MaterialFunctionCall(A=TextureCoordinate(), B=Constant2Vector() {g=0.464, r=0.606}, Bounds=Constant2Vector() {g=0.1375, r=0.13}, Edge Falloff=Constant() {r=0.01}) {materialFunction=&BoxMask-2D}
  n31 = Desaturation(None=NamedRerouteUsage(), Fraction=Constant() {r=0.9})
  n32 = ScalarParameter() {defaultValue=1, group=WarmShadeWagon, parameterName=WS_ScreenOverlayIntensity}
  n33 = Multiply(A=@n32) {constB=0.1}
  n34 = Divide(A=TextureCoordinate() {coordinateIndex=1}, B=Constant2Vector() {g=1, r=2})
  n35 = ComponentMask(None=@n34) {r=true}
  n36 = ComponentMask(None=@n34) {g=true}
  n37 = TextureSample!(UVs=AppendVector(A=Add(A=@n35) {constB=0}, B=@n36), Tex, Apply View MipBias) {
   samplerSource=SSM_Wrap_WorldGroupSettings
   samplerType=SAMPLERTYPE_LinearColor
   texture=&T_WarmShadeWagon_Backpack_mask
  }
  n38 = TextureSample(UVs=AppendVector(A=Add(A=@n35) {constB=0.499}, B=@n36)) {
   samplerSource=SSM_Wrap_WorldGroupSettings
   samplerType=SAMPLERTYPE_LinearColor
   texture=&T_WarmShadeWagon_Backpack_mask
  }
  n39 = LinearInterpolate(A=LinearInterpolate(A=@n37.R, B=@n38.R, Alpha=NamedRerouteUsage()) {constAlpha=0.4}, Alpha=NamedRerouteUsage()) {constB=0.5}
  n40 = NamedRerouteUsage()
  n41 = Add(
   A=Multiply(A=Multiply(A=@n31, B=@n31), B=QualitySwitch(Default=@n32, Low=@n33, High=@n32, Medium=@n32, Epic=@n32))
   B=MaterialFunctionCall(In=Add(A=@n39, B=@n40), Contrast=Constant() {r=-0.05}) {materialFunction=&CheapContrast}.Result
  )
  n42 = LinearInterpolate(
   A=VectorParameter() {group=WarmShadeWagon, parameterName=BackpackIdleColor, defaultValue=<rgba 0.1002 0.1567 0.8646 1>}.RGB
   B=VectorParameter() {group=WarmShadeWagon, parameterName=BackpackReactiveColor, defaultValue=<rgba 1 0 0.0065 1>}.RGB
   Alpha=NamedRerouteUsage()
  )
  n43 = Multiply(
   A=Multiply(
    A=NamedRerouteDeclaration(None=@n30.Result) {name=ScreenMask, nodeColor=<rgba 1 0 0.5156 1>}
    B=MaterialFunctionCall(Base=@n41, Blend=@n42) {materialFunction=&Blend_Overlay}.Result
   )
   B=ScalarParameter() {defaultValue=0.75, group=WarmShadeWagon, parameterName=WS_ScreenIntensity}
  ) {constB=0.7}
  n44 = Constant() {r=125}
  n45 = Multiply(
   A=@n43
   B=Subtract(A=MaterialFunctionCall!(ScanLine_Tiling_1, Scanline_PannerSpeed_1, ScanLine_Tiling=@n44, Scanline_PannerSpeed, UVs) {materialFunction=/Game/Weapons/WeaponSkins/MaterialFunction/ScanLines}.Result) {constB=0.5}
  )
  NamedRerouteDeclaration(None=Add(A=@n43, B=Multiply(A=@n45) {constB=0.5})) {name=Screen, nodeColor=<rgba 0 0.0156 1 1>}
  n46 = GetMaterialAttributes(None=NamedRerouteUsage()) {attributeGetTypes [
    #1
    #2
    #4
    #5
    #6
   ]}
  NamedRerouteDeclaration(None=@n46.CustomData0) {name=InClearCoat, nodeColor=<rgba 1 0 0.0469 1>}
  NamedRerouteDeclaration(None=@n46.Metallic) {name=InMetallic, nodeColor=<rgba 1 0.1172 0 1>}
  n47 = NamedRerouteUsage()
  n48 = Multiply(A=NamedRerouteUsage(), B=@n13)
  NamedRerouteDeclaration(None=StaticSwitchParameter(True=LinearInterpolate(B=@n47, Alpha=@n48) {constB=0}, False=NamedRerouteUsage()) {defaultValue=true, group=WarmShadeWagon, parameterName=UseSphereMask}) {name=OverrideCloth, nodeColor=<rgba 0 0.2031 1 1>}
  n49 = NamedRerouteUsage()
  NamedRerouteDeclaration(None=StaticSwitchParameter(True=LinearInterpolate(A=@n49, Alpha=@n48) {constB=0}, False=NamedRerouteUsage()) {defaultValue=true, group=WarmShadeWagon, parameterName=UseSphereMask}) {name=OverrideMetallic, nodeColor=<rgba 0.5234 1 0 1>}
  n50 = TextureSample() {
   constCoordinate=1
   samplerType=SAMPLERTYPE_LinearColor
   texture=&T_WarmShadeWagon_Backpack_mask
  }
  NamedRerouteDeclaration(None=@n50.RGB) {name=InBackpackMask, desc=BackpackMask, nodeColor=<rgba 1 0 0.1406 1>}
  n51 = NamedRerouteUsage()
  n52 = ScalarParameter() {defaultValue=0.95, group=WarmShadeWagon, parameterName=LogoOverrideRoughness}
  NamedRerouteDeclaration(None=StaticSwitchParameter(True=LinearInterpolate(A=@n51, B=@n52, Alpha=@n48) {constB=0.95}, False=NamedRerouteUsage()) {defaultValue=true, group=WarmShadeWagon, parameterName=UseSphereMask}) {name=OverrideRoughness, nodeColor=<rgba 0 0.9297 1 1>}
  NamedRerouteDeclaration(None=@n46.Roughness) {name=InRoughness, nodeColor=<rgba 1 0.75 0 1>}
  n53 = TextureSampleParameter2D() {
   desc=Replacing FXMask for better resolution
   group=WarmShadeWagon
   parameterName=WS_ReactivityMask
   samplerType=SAMPLERTYPE_Masks
   texture=&T_WhiteMask
  }
  n54 = ChannelMaskParameter(None=NamedRerouteUsage()) {group=WarmShadeWagon, maskChannel=Blue, parameterName=WS_ReactivityMaskChannel, defaultValue=<rgba 0 0 1 0>}
  n55 = Multiply(
   A=LinearInterpolate(A=@n53.R, B=@n54.R, Alpha=ScalarParameter() {defaultValue=1, group=WarmShadeWagon, parameterName=WS_ReactivityUseFXMask, sliderMax=1})
   B=ScalarParameter() {defaultValue=1, group=WarmShadeWagon, parameterName=WS_ReactivityMaskIntensity}
  )
  n56 = Fresnel(ExponentIn=ScalarParameter() {defaultValue=6, group=WarmShadeWagon, parameterName=WS_FresnelExp}, Normal=VertexNormalWS()) {baseReflectFraction=0}
  n57 = NamedRerouteUsage()
  n58 = Saturate(None=Multiply(A=Subtract(A=@n56, B=@n57), B=ScalarParameter() {defaultValue=5, group=WarmShadeWagon, parameterName=WS_FresnelIntensity}) {constB=5})
  n59 = ScalarParameter() {defaultValue=1, group=WarmShadeWagon, parameterName=WS_ReactExponentMin}
  n60 = ScalarParameter() {defaultValue=2, group=WarmShadeWagon, parameterName=WS_ReactExponentMax}
  n61 = Power(Base=NamedRerouteUsage()) {constExponent=4}
  n62 = Multiply(
   A=Power(Base=Add(A=@n55, B=@n58), Exp=LinearInterpolate(A=@n59, B=@n60, Alpha=@n61) {constA=1, constB=2})
   B=LinearInterpolate(
    A=ScalarParameter() {defaultValue=1, group=WarmShadeWagon, parameterName=WS_ReactIntensityMin}
    B=ScalarParameter() {defaultValue=2, group=WarmShadeWagon, parameterName=WS_ReactIntensityMax}
    Alpha=@n61
   ) {constA=1, constB=2}
  )
  n63 = Add(
   A=LocalPosition() {localOrigin=InstancePreSkinning}.XYZ
   B=VectorParameter() {group=WarmShadeWagon, parameterName=PreSkinnedPositionOffset, defaultValue=<rgba 0 0 -122 1>}.RGB
  )
  n64 = LinearInterpolate(Alpha=LinearInterpolate(
    A=NamedRerouteUsage()
    B=ScalarParameter() {defaultValue=1, desc=Controlled by BP, group=WarmShadeWagon, parameterName=AccumulatedSound}
    Alpha=ScalarParameter() {defaultValue=1, group=WarmShadeWagon, parameterName=UseAccumulatedSound}
   )) {constB=350}
  n65 = Multiply(A=SphereMask(A=VertexInterpolator(VS=@n63).PS, B=Constant(), Radius=Clamp(None=@n64) {maxDefault=130, minDefault=0.01}) {attenuationRadius=130, hardnessPercent=0}) {constB=5}
  n66 = LinearInterpolate(
   B=Multiply(A=@n62, B=NamedRerouteUsage())
   Alpha=NamedRerouteDeclaration(None=Saturate(None=@n65)) {name=ReactivityRadialMask, nodeColor=<rgba 0 1 0.3906 1>}
  )
  n67 = NamedRerouteUsage()
  n68 = NamedRerouteUsage()
  NamedRerouteDeclaration(None=StaticSwitchParameter(True=Multiply(A=@n66, B=@n67), False=Multiply(A=@n62, B=@n68)) {defaultValue=true, group=WarmShadeWagon, parameterName=UseSphereMask}) {name=ReactivityMask, nodeColor=<rgba 0 1 0.7422 1>}
  n69 = LinearInterpolate(
   A=Constant() {r=0.85}
   Alpha=MaterialFunctionCall(Speed=Constant() {r=2}) {materialFunction=&DebugTimeSine}.Result
  )
  n70 = ScalarParameter() {group=WarmShadeWagon, parameterName=ReactivityDebug, sliderMax=1}
  n71 = ScalarParameter() {defaultValue=1, group=WarmShadeWagon, parameterName=ReactivityOn, sliderMax=1}
  NamedRerouteDeclaration(None=LinearInterpolate(
    B=LinearInterpolate(A=@n14.Amplitude, B=@n69, Alpha=@n70)
    Alpha=NamedRerouteDeclaration(None=@n71) {name=ReactivityOn, nodeColor=<rgba 0 0.7422 1 1>}
   )) {name=Amplitude, nodeColor=<rgba 0 1 0.5547 1>}
  n72 = LocalPosition() {localOrigin=InstancePreSkinning}
  n73 = Saturate(None=VertexInterpolator(VS=Divide(A=@n72.Z) {constB=150}).PS)
  NamedRerouteDeclaration(None=Multiply(A=@n73, B=@n73)) {name=TopDownGradient, nodeColor=<rgba 0.8984 0 1 1>}
  NamedRerouteDeclaration(None=Normalize(VectorInput=ComponentMask(None=@n14.Spectrum) {b=true, g=true, r=true})) {name=NormalizedSpectrum, nodeColor=<rgba 0.5469 1 0 1>}
  NamedRerouteDeclaration(None=@n46.BaseColor) {name=InBaseColor, nodeColor=<rgba 0 1 0.0859 1>}
  NamedRerouteDeclaration(None=@n46.EmissiveColor) {name=InEmissive, nodeColor=<rgba 1 0 0.2344 1>}
  NamedRerouteUsage()
 ]}
