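# Parent material for the eyes of the WarmShadeWagon large male player body, built for skeletal meshes and mesh deformers.
# It passes the shared MF_eyes result through MF_WarmShadeWagon, which also takes the RGB of the FXMask texture parameter, defaulting to the head FX texture.
# MF_TechArtOverrides is applied last in this material-attributes chain, with shading model set from expression and resolving to Default Lit.
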Material /BRCosmetics/Characters/Player/Male/Large/Bodies/M_LRG_WarmShadeWagon/Materials/M_WarmShadeWagon_Eyes_Parent
graph {
 properties {
  bUseMaterialAttributes=true
  bUsedWithMeshDeformer=true
  bUsedWithSkeletalMesh=true
  pixelDepthOffsetMode=PDOM_Legacy
  shadingModel=MSM_FromMaterialExpression
  usedShadingModels=Default Lit
 }
 nodes [
  n1 = TextureSampleParameter2D() {
   group=WarmShadeWagon
   parameterName=FXMask
   samplerType=SAMPLERTYPE_Masks
   texture=/BRCosmetics/Characters/Player/Male/Large/Bodies/M_LRG_WarmShadeWagon/Textures/T_WarmShadeWagon_Head_FX
  }
  MaterialFunctionCall(In=MaterialFunctionCall!(In=MaterialFunctionCall() {materialFunction=/Game/Characters/Shared/FNEye/Materials/MF_eyes}.Result, FXMask=@n1.RGB) {materialFunction=&/MF_WarmShadeWagon}.Result) {materialFunction=&MF_TechArtOverrides}
 ]
}
