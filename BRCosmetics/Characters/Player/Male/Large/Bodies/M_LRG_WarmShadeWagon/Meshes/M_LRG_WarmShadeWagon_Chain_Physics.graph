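# Physics asset for the chain on the WarmShadeWagon large male player body cosmetic, letting a chain hung from the upper spine swing against the torso.
# Four dyn_chain bodies form a serial chain from spine_05, with capsules for the first three and a box at the end.
# Capsules on spine_04 and spine_05 approximate the torso.
# Joint limits restrict sideways swing and twist, with swing2 locked on the upper two joints and twist locked at the root.
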PhysicsAsset /BRCosmetics/Characters/Player/Male/Large/Bodies/M_LRG_WarmShadeWagon/Meshes/M_LRG_WarmShadeWagon_Chain_Physics
bodies {
 dyn_chain_01 [{
   shapeType=Capsule
   length=13.0696
   radius=1.2244
   shapeName=dyn_chain_capsule
   center=<0.5052 6.5782 -0.0135>
   rotation=<rot 0.3199 -2.6981 -90.0799>
  }]
 dyn_chain_02 [{
   shapeType=Capsule
   length=4.8695
   radius=1.3244
   shapeName=dyn_chain_capsule
   center=<0.0831 2.9102 -0.0542>
   rotation=<rot 0.0267 -3.1766 -89.5204>
  }]
 dyn_chain_03 [{
   shapeType=Capsule
   length=1.8695
   radius=0.8244
   shapeName=dyn_chain_capsule
   center=<-0.2412 0.1113 -0.0361>
   rotation=<rot 0.2749 5.7001 -90.9202>
  }]
 dyn_chain_04 [{
   shapeType=Box
   extentX=3.628
   extentY=19.5042
   extentZ=8.8196
   shapeName=dyn_chain_04_box
   center=<0.4551 3.7733 0.0202>
   rotation=<rot -0.1108 5.5628 -89.4015>
  }]
 spine_04 [
  {shapeType=Capsule, length=19.1956, radius=14.0016, center=<5.0572 1.0616 0.1525>, rotation=<rot 0.2785 3.4327 0.3147>}
  {shapeType=Capsule, length=19.3, radius=10.7, shapeName=spine_04_capsule, center=<9.3957 2.8303 0>}
 ]
 spine_05 [{shapeType=Capsule, length=28.3, radius=9.5818, center=<3.304 -1.3178 -0.1411>, rotation=<rot 0.1916 93.0032 179.43>}]
}
constraints [
 {bone1Name=spine_05, bone2Name=spine_04, swing1LimitDegrees=45, swing2LimitDegrees=45, twistLimitDegrees=45}
 {bone1Name=dyn_chain_04, bone2Name=dyn_chain_03, swing1LimitDegrees=35, swing2LimitDegrees=3, twistLimitDegrees=5}
 {bone1Name=dyn_chain_03, bone2Name=dyn_chain_02, swing1LimitDegrees=35, swing2LimitDegrees=3, twistLimitDegrees=5}
 {bone1Name=dyn_chain_02, bone2Name=dyn_chain_01, swing1LimitDegrees=35, swing2LimitDegrees=10, swing2Motion=Locked, twistLimitDegrees=20}
 {
  bone1Name=dyn_chain_01
  bone2Name=spine_05
  swing1LimitDegrees=25
  swing2LimitDegrees=10
  swing2Motion=Locked
  twistLimitDegrees=45
  twistMotion=Locked
 }
]
