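# Physics asset for the dangling chain on the WarmShadeWagon large male player body, the FrontEnd variant by its name.
# Three capsule links and a larger box at the free end form a four-body chain, constrained in series to spine_05, whose two capsules approximate the upper torso.
# Angular limits restrict motion mostly to Swing1, with Swing2 locked on the upper two joints and twist also locked at the spine attachment.
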PhysicsAsset /BRCosmetics/Characters/Player/Male/Large/Bodies/M_LRG_WarmShadeWagon/Meshes/M_LRG_WarmShadeWagon_Chain_Physics_FrontEnd
bodies {
 dyn_chain_01 [{
   shapeType=Capsule
   length=13.0696
   radius=1.2244
   shapeName=dyn_chain_capsule
   center=<0.5052 6.5782 -0.0135>
   rotation=<rot 0.3199 -2.6981 -90.0799>
  }]
 dyn_chain_02 [{
   shapeType=Capsule
   length=5.0695
   radius=1.1244
   shapeName=dyn_chain_capsule
   center=<0.0831 2.9102 -0.0542>
   rotation=<rot 0.0267 -3.1766 -89.5204>
  }]
 dyn_chain_03 [{
   shapeType=Capsule
   length=0.8695
   radius=1.1244
   shapeName=dyn_chain_capsule
   center=<-0.0314 0.1323 -0.0351>
   rotation=<rot 0.2749 5.7001 -90.9202>
  }]
 dyn_chain_04 [{
   shapeType=Box
   extentX=3.628
   extentY=15.7042
   extentZ=7.4196
   shapeName=dyn_chain_04_box
   center=<0.71 3.5679 0.0221>
   rotation=<rot -0.0755 2.2068 -89.3961>
  }]
 spine_05 [
  {shapeType=Capsule, length=28.3, radius=9.5818, center=<3.304 -1.3178 -0.1411>, rotation=<rot 0.1916 93.0032 179.43>}
  {shapeType=Capsule, length=20.2, radius=9.8, shapeName=spine_05_capsule, center=<-2.8247 4.4999 0>}
 ]
}
constraints [
 {bone1Name=dyn_chain_04, bone2Name=dyn_chain_03, swing1LimitDegrees=35, swing2LimitDegrees=3, twistLimitDegrees=5}
 {bone1Name=dyn_chain_03, bone2Name=dyn_chain_02, swing1LimitDegrees=35, swing2LimitDegrees=3, twistLimitDegrees=5}
 {bone1Name=dyn_chain_02, bone2Name=dyn_chain_01, swing1LimitDegrees=35, swing2LimitDegrees=10, swing2Motion=Locked, twistLimitDegrees=20}
 {
  bone1Name=dyn_chain_01
  bone2Name=spine_05
  swing1LimitDegrees=25
  swing2LimitDegrees=10
  swing2Motion=Locked
  twistLimitDegrees=45
  twistMotion=Locked
 }
]
