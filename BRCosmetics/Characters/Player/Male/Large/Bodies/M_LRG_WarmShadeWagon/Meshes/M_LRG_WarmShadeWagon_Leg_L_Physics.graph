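# Physics asset for the left leg mesh of the large male player body M_LRG_WarmShadeWagon, with collision for both legs and dynamic shorts bones.
# Capsule bodies cover both thighs and calves, plus thin capsules on front, back, left and right left-side shorts bones.
# Each shorts body is constrained to the left thigh with uneven swing and twist limits, one tight axis apiece, while each calf is constrained to its thigh.
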PhysicsAsset /BRCosmetics/Characters/Player/Male/Large/Bodies/M_LRG_WarmShadeWagon/Meshes/M_LRG_WarmShadeWagon_Leg_L_Physics
bodies {
 calf_l [{shapeType=Capsule, length=26.6468, radius=6.3478, center=<-18.9017 -2.7231 0.7443>, rotation=<rot 1.8705 88.4484 -91.9232>}]
 calf_r [{shapeType=Capsule, length=26.6468, radius=6.3478, center=<18.997 2.549 -0.6044>, rotation=<rot 1.8705 88.4484 -91.9232>}]
 dyn_shorts_back_l [{
   shapeType=Capsule
   length=9.0921
   radius=1.3461
   shapeName=dyn_shorts_right_r_capsule
   center=<-5.2633 -0.0268 0.0323>
   rotation=<rot 86.6068 -114.0949 -114.3028>
  }]
 dyn_shorts_front_l [{
   shapeType=Capsule
   length=18.1922
   radius=1.0461
   shapeName=dyn_shorts_right_r_capsule
   center=<-0.0859 -0.2983 8.9369>
   rotation=<rot -0.5013 -173.7354 -177.9712>
  }]
 dyn_shorts_left_l [{
   shapeType=Capsule
   length=7.7921
   radius=1.2461
   shapeName=dyn_shorts_right_r_capsule
   center=<-0.1928 0.3657 4.36>
   rotation=<rot -5.7578 -179.1671 -179.2276>
  }]
 dyn_shorts_right_l [{
   shapeType=Capsule
   length=10.2921
   radius=1.2461
   shapeName=dyn_shorts_right_r_capsule
   center=<-0.0343 -0.011 5.9551>
   rotation=<rot 0.6139 0.7575 -0.9485>
  }]
 thigh_l [{shapeType=Capsule, length=38.7468, radius=6.1478, center=<-24.0091 -2.1861 2.7753>, rotation=<rot 0.5371 92.8404 -85.3357>}]
 thigh_r [{shapeType=Capsule, length=37.6468, radius=7.2478, center=<24.1916 1.9252 -2.1982>, rotation=<rot 0.5371 92.8404 -85.3357>}]
}
constraints [
 {bone1Name=calf_l, bone2Name=thigh_l, swing1LimitDegrees=45, swing2LimitDegrees=45, twistLimitDegrees=45}
 {bone1Name=calf_r, bone2Name=thigh_r, swing1LimitDegrees=45, swing2LimitDegrees=45, twistLimitDegrees=45}
 {bone1Name=dyn_shorts_right_l, bone2Name=thigh_l, swing1LimitDegrees=3, swing2LimitDegrees=30, twistLimitDegrees=25}
 {bone1Name=dyn_shorts_left_l, bone2Name=thigh_l, swing1LimitDegrees=3, swing2LimitDegrees=45, twistLimitDegrees=25}
 {bone1Name=dyn_shorts_front_l, bone2Name=thigh_l, swing1LimitDegrees=3, swing2LimitDegrees=25, twistLimitDegrees=27}
 {bone1Name=dyn_shorts_back_l, bone2Name=thigh_l, swing1LimitDegrees=36, swing2LimitDegrees=15, twistLimitDegrees=3}
]
