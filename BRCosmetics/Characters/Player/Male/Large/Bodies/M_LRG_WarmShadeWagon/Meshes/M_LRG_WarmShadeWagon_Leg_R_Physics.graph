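# Physics asset for the right-leg mesh of the large male player body cosmetic WarmShadeWagon, covering both legs and the right shorts' dynamic bones.
# Single capsules approximate each thigh and calf, with each calf constrained to its thigh.
# Four thin capsules on front, back, left and right shorts bones each constrain to the right thigh.
# Their angular limits are asymmetric, each tight on one axis and looser on the others.
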PhysicsAsset /BRCosmetics/Characters/Player/Male/Large/Bodies/M_LRG_WarmShadeWagon/Meshes/M_LRG_WarmShadeWagon_Leg_R_Physics
bodies {
 calf_l [{shapeType=Capsule, length=26.6468, radius=6.3478, center=<-19.1239 -2.4115 0.7864>, rotation=<rot 1.8294 87.9483 -91.9268>}]
 calf_r [{shapeType=Capsule, length=26.6468, radius=6.3478, center=<18.997 2.549 -0.6044>, rotation=<rot 1.8705 88.4484 -91.9232>}]
 dyn_shorts_back_r [{
   shapeType=Capsule
   length=8.3921
   radius=1.3461
   shapeName=dyn_shorts_right_r_capsule
   center=<4.7202 -0.0657 0.2711>
   rotation=<rot 86.0283 -110.375 -110.1965>
  }]
 dyn_shorts_front_r [{
   shapeType=Capsule
   length=18.3922
   radius=1.0461
   shapeName=dyn_shorts_right_r_capsule
   center=<0.1029 0.2792 -9.0162>
   rotation=<rot -0.5641 -177.5361 -177.9425>
  }]
 dyn_shorts_left_r [{
   shapeType=Capsule
   length=9.7921
   radius=1.2461
   shapeName=dyn_shorts_right_r_capsule
   center=<-0.1074 0.1795 -5.2413>
   rotation=<rot -0.0691 178.9262 -179.0327>
  }]
 dyn_shorts_right_r [{
   shapeType=Capsule
   length=7.7921
   radius=1.2461
   shapeName=dyn_shorts_right_r_capsule
   center=<0.4593 0.4952 -4.0332>
   rotation=<rot -7.7297 178.7961 -179.0238>
  }]
 thigh_l [{shapeType=Capsule, length=24.8468, radius=8.1478, center=<-23.1915 -1.7532 2.1812>, rotation=<rot 0.7627 90.1103 -85.3668>}]
 thigh_r [{shapeType=Capsule, length=38.7468, radius=6.1478, center=<24.1916 1.9252 -2.1982>, rotation=<rot 0.5371 92.8404 -85.3357>}]
}
constraints [
 {bone1Name=calf_l, bone2Name=thigh_l, swing1LimitDegrees=45, swing2LimitDegrees=45, twistLimitDegrees=45}
 {bone1Name=calf_r, bone2Name=thigh_r, swing1LimitDegrees=45, swing2LimitDegrees=45, twistLimitDegrees=45}
 {bone1Name=dyn_shorts_right_r, bone2Name=thigh_r, swing1LimitDegrees=3, swing2LimitDegrees=45, twistLimitDegrees=25}
 {bone1Name=dyn_shorts_back_r, bone2Name=thigh_r, swing1LimitDegrees=36, swing2LimitDegrees=15, twistLimitDegrees=3}
 {bone1Name=dyn_shorts_front_r, bone2Name=thigh_r, swing1LimitDegrees=3, swing2LimitDegrees=25, twistLimitDegrees=27}
 {bone1Name=dyn_shorts_left_r, bone2Name=thigh_r, swing1LimitDegrees=3, swing2LimitDegrees=30, twistLimitDegrees=25}
]
