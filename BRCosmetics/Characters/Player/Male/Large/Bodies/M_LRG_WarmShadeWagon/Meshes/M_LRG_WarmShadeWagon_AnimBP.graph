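# Post-copy AnimBlueprint for the WarmShadeWagon large male body cosmetic, adding secondary physics and cloth-like material motion.
# The parent pose feeds RigidBody chains for spine_05 and both legs, blended between front-end and in-game setups, then a Control Rig receiving LOD and physics-off flags.
# The event graph drives per-leg material WPO from bone transforms.
# Skydiving lowers leg physics alpha and raises material wind and turbulence.
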AnimBlueprint /BRCosmetics/Characters/Player/Male/Large/Bodies/M_LRG_WarmShadeWagon/Meshes/M_LRG_WarmShadeWagon_AnimBP
target_skeleton=&/M_LRG_WarmShadeWagon_Skeleton
graphs [
 AnimGraph {nodes [
   n1 = RigidBody(ComponentPose=Local To Component(LocalPose=Use cached pose 'PreSim'())) {
    bOverrideWorldGravity=true
    overridePhysicsAsset=&/M_LRG_WarmShadeWagon_Chain_Physics_FrontEnd
    simulationSpace=BaseBoneSpace
    alphaBoolBlend {blendOption=Linear}
    baseBoneRef {boneName=spine_05}
    componentAppliedLinearAccClamp=<2500 5000 5000>
    componentLinearAccScale {x=0.27}
    componentLinearVelScale {x=0.25}
    overrideWorldGravity {z=-980}
    simSpaceSettings {maxAngularAcceleration=500, maxAngularVelocity=10, maxLinearAcceleration=350, maxLinearVelocity=100, velocityScaleZ=0, worldAlpha=0.35}
   }
   n2 = Get Alpha Rigid Body() -> (Alpha Rigid Body:double)
   n3 = RigidBody(ComponentPose=@n1, Alpha=@n2) {
    bOverrideWorldGravity=true
    overridePhysicsAsset=&/M_LRG_WarmShadeWagon_Leg_R_Physics
    simulationSpace=BaseBoneSpace
    alphaBoolBlend {blendOption=Linear}
    baseBoneRef {boneName=thigh_r}
    componentAppliedLinearAccClamp=<8000 8000 8000>
    componentLinearAccScale=<0.6 0.6 0.6>
    componentLinearVelScale=<0.1 0.1 0.1>
    overrideWorldGravity {z=-980}
    simSpaceSettings {maxAngularAcceleration=500, maxAngularVelocity=10, maxLinearAcceleration=350, maxLinearVelocity=100, velocityScaleZ=0, worldAlpha=0.55}
   }
   n4 = RigidBody(ComponentPose=@n3, Alpha=@n2) {
    bOverrideWorldGravity=true
    overridePhysicsAsset=&/M_LRG_WarmShadeWagon_Leg_L_Physics
    simulationSpace=BaseBoneSpace
    alphaBoolBlend {blendOption=Linear}
    baseBoneRef {boneName=thigh_l}
    componentAppliedLinearAccClamp=<8000 8000 8000>
    componentLinearAccScale=<0.6 0.6 0.6>
    componentLinearVelScale=<0.1 0.1 0.1>
    overrideWorldGravity {z=-980}
    simSpaceSettings {maxAngularAcceleration=500, maxAngularVelocity=10, maxLinearAcceleration=350, maxLinearVelocity=100, velocityScaleZ=0, worldAlpha=0.55}
   }
   n5 = RigidBody(ComponentPose=Local To Component(LocalPose=Use cached pose 'PreSim'())) {
    bOverrideWorldGravity=true
    overridePhysicsAsset=&/M_LRG_WarmShadeWagon_Chain_Physics
    simulationSpace=BaseBoneSpace
    alphaBoolBlend {blendOption=Linear}
    baseBoneRef {boneName=spine_05}
    componentAppliedLinearAccClamp=<2500 5000 5000>
    componentLinearAccScale {x=0.27}
    componentLinearVelScale {x=0.25}
    overrideWorldGravity {z=-980}
    simSpaceSettings {maxAngularAcceleration=500, maxAngularVelocity=10, maxLinearAcceleration=350, maxLinearVelocity=100, velocityScaleZ=0, worldAlpha=0.5}
   }
   n6 = RigidBody(ComponentPose=@n5, Alpha=@n2) {
    bOverrideWorldGravity=true
    overridePhysicsAsset=&M_LRG_WarmShadeWagon_Leg_R_Physics
    simulationSpace=BaseBoneSpace
    alphaBoolBlend {blendOption=Linear}
    baseBoneRef {boneName=thigh_r}
    componentAppliedLinearAccClamp=<8000 8000 8000>
    componentLinearAccScale=<0.6 0.6 0.6>
    componentLinearVelScale=<0.1 0.1 0.1>
    overrideWorldGravity {z=-980}
    simSpaceSettings {maxAngularAcceleration=500, maxAngularVelocity=10, maxLinearAcceleration=350, maxLinearVelocity=100, velocityScaleZ=0, worldAlpha=0.55}
   }
   n7 = RigidBody(ComponentPose=@n6, Alpha=@n2) {
    bOverrideWorldGravity=true
    overridePhysicsAsset=&M_LRG_WarmShadeWagon_Leg_L_Physics
    simulationSpace=BaseBoneSpace
    alphaBoolBlend {blendOption=Linear}
    baseBoneRef {boneName=thigh_l}
    componentAppliedLinearAccClamp=<8000 8000 8000>
    componentLinearAccScale=<0.6 0.6 0.6>
    componentLinearVelScale=<0.1 0.1 0.1>
    overrideWorldGravity {z=-980}
    simSpaceSettings {maxAngularAcceleration=500, maxAngularVelocity=10, maxLinearAcceleration=350, maxLinearVelocity=100, velocityScaleZ=0, worldAlpha=0.55}
   }
   n8 = Control Rig!(
    Source=Blend Poses by bool(BlendPose_0=Component To Local(ComponentPose=@n4), BlendPose_1=Component To Local(ComponentPose=@n7), bActiveValue=Get bIsInFrontEnd()) {blendType=Linear}
    Alpha
    LOD1:bool=Get LOD1() -> (LOD1:bool)
    LOD2:bool=Get LOD2() -> (LOD2:bool)
    LOD3:bool=Get LOD3() -> (LOD3:bool)
    RBAN off:bool=Get RBAN_off() -> (RBAN_off:bool)
   ) -> (Pose) {
    alphaBoolBlend {blendOption=Linear}
    controlRigAssetReference {blueprintRigClass=&/M_LRG_WarmShadeWagon_CtrlRig.M_LRG_WarmShadeWagon_CtrlRig_C}
   }
   Output Pose(Result=@n8) {layerGroup=None}
   Save cached pose 'PreSim'(Pose=Copy Pose From Mesh() {bUseAttachedParent=true}) {cacheName=PreSim}
  ]}
 EventGraph {nodes [
   Event BlueprintUpdateAnimation()
   > n1 = Sequence()
   > Is Valid(InputObject=Get Player Pawn() -> (Player Pawn:Fort Player Pawn Athena Object Reference))
   > Set WPO Globals Function(
    Material=Get Material Instance Dynamic() -> (Material Instance Dynamic:Material Instance Dynamic Object Reference)
    Overlay Material=Get Overlay Material Instance Dynamic() -> (Overlay Material Instance Dynamic:Material Instance Dynamic Object Reference)
    Skeletal Mesh=Get Skeletal Mesh() -> (Skeletal Mesh:Skeletal Mesh Component Object Reference)
    "Local Noise Transform  Location"="Get Local Noise Transform  Location"() -> ("Local Noise Transform  Location":Vector)
    Local Noise Offset=Get Local Noise Offset()
   )
   > Set WPO Regional Function!(
    execute
    Material:Material Instance Dynamic Object Reference=Get Material Instance Dynamic()
    Overlay Material:Material Instance Dynamic Object Reference=Get Overlay Material Instance Dynamic()
    Skeletal Mesh:Skeletal Mesh Component Object Reference=Get Skeletal Mesh()
    Velocity Transform:Name=calf_r
    Velocity Interp Speed:double=Get Velocity Interp Speed()
    Collision Transform:Name=thigh_r
    Collision Transform Axis:EAxis Enum
    Invert Collision Transform Axis:bool=false
    Delta Time X:double=0
    Previous Velocity Transform Location:"Vector (by ref)"=Get R1 Velocity Transform Location()
    Previous Velocity Transform Velocity:"Vector (by ref)"=Get R1 Velocity Transform Velocity()
    Region:int=1
   ) -> (then)
   > Set WPO Regional Function(
    Material=Get Material Instance Dynamic()
    Overlay Material=Get Overlay Material Instance Dynamic()
    Skeletal Mesh=Get Skeletal Mesh()
    Velocity Transform=calf_l
    Velocity Interp Speed=Get Velocity Interp Speed()
    Collision Transform=thigh_l
    Invert Collision Transform Axis=true
    Delta Time X=0
    Previous Velocity Transform Location=Get R2 Velocity Transform Location()
    Previous Velocity Transform Velocity=Get R2 Velocity Transform Velocity()
    Region=2
   )
   > n2 = Set Skydive Enabled(Skydive Enabled:bool=OR Boolean(A=Get bIsSkydiveFloating(), B=Get bIsSkydiveDiveMode())) -> (Output_Get:bool)
   > n3 = Branch(Condition=@n2)
   > Set Alpha Rigid Body(Alpha Rigid Body:double=0.2) -> (Output_Get:double)
   > Set Turbulence Noise(Turbulence Noise:float=75) -> (Output_Get:float)
   > n4 = Set WPO Pho Wind Intensity(WPO Pho Wind Intensity=1)
   n5 = Get Material Instance Dynamic()
   n6 = Get Overlay Material Instance Dynamic()
   SetScalarParameterValue(execute=[@n7.then, @n4.then], self=[@n5, @n6], ParameterName=ClothSim_WindStrength, Value=Get WPO Pho Wind Intensity())
   > SetScalarParameterValue(self=[@n5, @n6], ParameterName=ClothSim_TurbulenceNoiseTileDivisor, Value=Get Turbulence Noise() -> (Turbulence Noise:float))
   Set Alpha Rigid Body(execute=@n3.else, Alpha Rigid Body=0.55)
   > Set Turbulence Noise(Turbulence Noise=0)
   > n7 = Set WPO Pho Wind Intensity(WPO Pho Wind Intensity=0)
   n8 = Sequence!(execute=@n1.then_1) -> (then_0, then_1, then_2:Exec, then_3:Exec)
   n9 = Get CurrentLOD()
   > Set LOD3(LOD3:bool="integer >= integer"(A=@n9, B=3)) -> (Output_Get:bool)
   Set LOD2(execute=@n8.then_1, LOD2:bool="integer >= integer"(A=@n9, B=2)) -> (Output_Get:bool)
   Set LOD1(execute=@n8.then_2, LOD1:bool="integer >= integer"(A=@n9, B=1)) -> (Output_Get:bool)
   Set RBAN_off(execute=@n8.then_3, RBAN_off:bool=NOT Boolean(A=Get bRigidBodyEnabled())) -> (Output_Get:bool)
   Event BlueprintInitializeAnimation()
   > Set Skeletal Mesh(Skeletal Mesh:Skeletal Mesh Component Object Reference=GetOwnerSkeletalMeshComponent()) -> (Output_Get:Skeletal Mesh Component Object Reference)
   n10 = TryGetPawnOwner()
   > Is Valid(InputObject=@n10)
   > n11 = Cast To FortPlayerPawnAthena(Object=@n10)
   > Set Player Pawn(Player Pawn:Fort Player Pawn Athena Object Reference=@n11) -> (Output_Get:Fort Player Pawn Athena Object Reference)
   Event OnCharacterCustomizationLoaded()
   > n12 = Set Skeletal Mesh(Skeletal Mesh=GetPartSkeletalMeshComponent())
   > n13 = Get Dynamic Material Instance(SkeletalMeshComponent=@n12)
   > Set Material Instance Dynamic(Material Instance Dynamic:Material Instance Dynamic Object Reference=@n13) -> (Output_Get:Material Instance Dynamic Object Reference)
   > n14 = GetOrCreateDynamicOverlayMaterial(MeshComponent=@n12, MaterialName=MI_WarmShadeWagon_Body)
   > Set Overlay Material Instance Dynamic(Overlay Material Instance Dynamic:Material Instance Dynamic Object Reference=@n14) -> (Output_Get:Material Instance Dynamic Object Reference)
   Get bIsParachuteOpened()
  ]}
]
variables [
 Alpha Rigid Body {type=number, default=0.55}
 Local Noise Offset {type=Vector}
 "Local Noise Transform  Location" {type=Vector}
 LOD1 {type=boolean}
 LOD2 {type=boolean}
 LOD3 {type=boolean}
 Material Instance Dynamic {type=MaterialInstanceDynamic}
 Overlay Material Instance Dynamic {type=MaterialInstanceDynamic}
 Player Pawn {type=FortPlayerPawnAthena}
 R1 Velocity Transform Location {type=Vector}
 R1 Velocity Transform Velocity {type=Vector}
 R2 Velocity Transform Location {type=Vector}
 R2 Velocity Transform Velocity {type=Vector}
 RBAN_off {type=boolean}
 Skeletal Mesh {type=SkeletalMeshComponent}
 Skydive Enabled {type=boolean}
 Turbulence Noise {type=number}
 Velocity Interp Speed {type=number}
 WPO Pho Wind Intensity {type=number, default=1}
]
