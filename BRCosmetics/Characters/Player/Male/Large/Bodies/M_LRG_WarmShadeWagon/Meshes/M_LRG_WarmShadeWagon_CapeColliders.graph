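# Physics asset of cape colliders for the large male player body M_LRG_WarmShadeWagon, giving the cape a simplified body volume to collide with.
# Capsules cover the pelvis, spine, clavicles and upper neck, with extra overlapping capsules on spine_02 and spine_05, and a sphere covers the head.
# Arm, thigh, calf and foot bodies list no shapes, and the asset has no constraints, so it defines collision volumes without ragdoll joints.
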PhysicsAsset /BRCosmetics/Characters/Player/Male/Large/Bodies/M_LRG_WarmShadeWagon/Meshes/M_LRG_WarmShadeWagon_CapeColliders
bodies {
 calf_l []
 calf_r []
 clavicle_l [{shapeType=Capsule, length=10.0192, radius=9.8871, center=<10.4225 -1.1829 1.5985>, rotation=<rot 76.3345 -0.1316 175.4345>}]
 clavicle_r [{shapeType=Capsule, length=9.6192, radius=9.8871, center=<-10.8269 1.1488 -1.4034>, rotation=<rot 76.2355 -0.0988 -4.5336>}]
 foot_l []
 foot_r []
 head [{shapeType=Sphere, radius=17.3125, center=<11.028 -0.174 0>}]
 lowerarm_l []
 lowerarm_r []
 neck_02 [{shapeType=Capsule, length=13.4375, radius=7.3175, center=<-0.2642 -0.3408 0>, rotation=<rot 90 -171.1193 -175.3978>}]
 pelvis [{shapeType=Capsule, length=9.2456, radius=15.3388, center=<-4.1903 -1.1025 0>}]
 spine_01 [{shapeType=Capsule, length=10.6438, radius=14.3586, center=<-1.7965 0.1884 0>}]
 spine_02 [
  {shapeType=Capsule, length=7.3264, radius=14.4322, center=<-1.711 -0.5396 0.0001>}
  {shapeType=Capsule, length=29.3956, radius=11.5807, center=<2.5024 0.4992 3.9825>, rotation=<rot 90 0.0049 0.0049>}
  {shapeType=Capsule, length=29.875, radius=11.1013, center=<2.4704 0.4991 -3.9825>, rotation=<rot 90 -0.0396 -0.0396>}
 ]
 spine_03 [{shapeType=Capsule, length=9.4107, radius=14.7031, center=<2.5296 -0.324 0.0001>}]
 spine_04 [{shapeType=Capsule, length=13.6404, radius=15.3326, center=<2.766 0.4039 0.0002>}]
 spine_05 [
  {shapeType=Capsule, length=16.6226, radius=16.4951, center=<-0.2772 1.1412 0>}
  {shapeType=Capsule, length=13.9518, radius=13.9425, center=<1.8873 0.4452 -0.0978>}
 ]
 thigh_l []
 thigh_r []
 upperarm_l []
 upperarm_r []
}
constraints []
